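# Translucent, unlit, two-sided mesh-particle material, likely for the SherWolf Baron boss's slash VFX.
# Two panning noise samples plus a scrolling gradient, cut on lower feature levels, are shaped by gradient and optional vertex-color masks into opacity.
# The dynamic Swipe parameter slides a sheen band that masks opacity and blends secondary to main color, tinted by particle color and scaled by EmissiveScale.
# Optional Fresnel, near-camera fade and quality-switched depth fade soften it.
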Material /VividRazor_Baron/Assets/VFX/Materials/M_SherWolf_Baron_Boss_Slash
graph {
 properties {
  bUsedWithMeshParticles=true
  bUsedWithNiagaraMeshParticles=true
  blendMode=BLEND_Translucent
  pixelDepthOffsetMode=PDOM_Legacy
  shadingModel=MSM_Unlit
  twoSided=true
  usedShadingModels=Unlit
 }
 nodes [
  n1 = DotProduct(A=VertexNormalWS(), B=CameraVectorWS())
  n2 = Multiply(A=Time(), B=ScalarParameter() {defaultValue=1, parameterName=TimeScale})
  n3 = Panner(Coordinate=TextureCoordinate(), Time=@n2) {speedY=-2}
  n4 = Multiply(A=Power(Base=TextureSample(UVs=@n3) {samplerSource=SSM_Wrap_WorldGroupSettings, texture=/Game/Global/Textures/Default/Masks_and_Gradients/CylGradient}.G) {constExponent=4}) {constB=60}
  n5 = Constant()
  n6 = TextureCoordinate()
  n7 = VectorParameter() {parameterName=UVs, defaultValue=<rgba 4 0.5 1 0.2>}
  n8 = AppendVector(A=@n7.R, B=@n7.R)
  n9 = TextureSampleParameter2D(UVs=Panner(Coordinate=Multiply(A=@n6, B=@n8), Time=@n2) {speedY=-1}) {parameterName=Tex1, texture=&T_NoisePack_01}
  n10 = TextureCoordinate()
  n11 = AppendVector(A=@n7.B, B=@n7.B)
  n12 = TextureSampleParameter2D(UVs=Panner(Coordinate=Multiply(A=@n10, B=@n11), Time=@n2) {speedY=-1}) {parameterName=Tex2, texture=&T_NoisePack_01}
  n13 = Constant()
  n14 = Add(
   A=FeatureLevelSwitch(Default=@n4, ES2=@n5, ES3_1=@n5, SM4_REMOVED=@n5, SM5=@n4, SM6=@n4)
   B=Add(
    A=LinearInterpolate(Alpha=@n9.R) {constA=0.15}
    B=FeatureLevelSwitch(Default=@n12.G, ES2=@n13, ES3_1=@n13, SM4_REMOVED=@n13, SM5=@n12.G, SM6=@n12.G)
   )
  )
  n15 = MaterialFunctionCall() {materialFunction=&LinearGradient}
  n16 = Multiply(
   A=@n15.VGradient
   B=Power(Base=OneMinus(None=@n15.VGradient), Exp=ScalarParameter() {defaultValue=4, parameterName=GradientPower}) {constExponent=3}
  )
  n17 = Multiply(A=ParticleColor().A, B=Multiply(A=@n14, B=Multiply(A=@n16) {constB=2}))
  n18 = ScalarParameter() {defaultValue=2, parameterName=Power}
  n19 = Multiply(A=@n17, B=StaticSwitchParameter(True=Power(Base=@n17, Exp=@n18), False=Constant() {r=1}) {parameterName=Use_Power})
  n20 = StaticSwitchParameter(True=Multiply(A=VertexColor().R, B=@n19), False=@n19) {defaultValue=true, parameterName=Use Vert Color}
  n21 = Multiply(A=Power(Base=Abs(None=@n1), Exp=ScalarParameter() {defaultValue=2, parameterName=Fresnel Power}), B=@n20)
  n22 = Constant() {r=1}
  n23 = Multiply(A=DynamicParameter() {defaultValue=<rgba 0.5 1 1 1>, paramNames [Swipe, EmissiveScale, Param3, Param4]}.Swipe, B=Add(A=@n22))
  n24 = ComponentMask(None=Add(A=TextureCoordinate()) {constB=0.5}) {g=true}
  n25 = Divide(A=Subtract(A=OneMinus(None=@n24), B=Subtract(A=@n23, B=@n22)), B=@n22)
  n26 = Saturate(None=ScalarParameter() {defaultValue=0.8, group=Sheen, parameterName=Sheen Center Point, sliderMax=1})
  n27 = Subtract(A=Saturate(None=Saturate(None=@n25) {desc=Remapped gradient}), B=@n26)
  n28 = Subtract(B=@n26)
  n29 = Multiply(A=Multiply(A=Divide(B=@n26), B=@n27)) {constB=-1}
  n30 = Max(A=Multiply(A=@n27, B=Divide(B=@n28)), B=@n29)
  n31 = Power(
   Base=OneMinus(None=Saturate(None=@n30))
   Exp=ScalarParameter() {defaultValue=5, group=Sheen, parameterName=Sheen Exponent, sliderMax=9, sliderMin=1}
  )
  n32 = Multiply(
   A=MaterialFunctionCall(Distance=ScalarParameter() {defaultValue=300, parameterName=Near Camera Fade}) {materialFunction=&Near_Camera_Fade}.Result
   B=Multiply(A=StaticSwitchParameter(True=@n21, False=@n20) {parameterName=Use Fresnel}, B=Multiply(B=@n31) {constA=1})
  )
  n33 = DepthFade(Opacity=@n32, FadeDistance=ScalarParameter() {defaultValue=125, parameterName=DepthFadeDistance})
  n34 = QualitySwitch(Default=@n33, Low=@n32, High=@n33, Medium=@n32, Epic=@n33)
  FeatureLevelSwitch(Default=@n34, ES2=@n32, ES3_1=@n32, SM4_REMOVED=@n32, SM5=@n34, SM6=@n34)
  n35 = ParticleColor()
  n36 = OneMinus(None=Saturate(None=@n29))
  n37 = Max(A=Multiply(A=Power(Base=@n36) {constExponent=24}) {constB=0.17}, B=Power(Base=@n36) {constExponent=33})
  n38 = LinearInterpolate(A=LinearInterpolate(A=@n35.RGB) {constAlpha=0.7}, B=@n35.RGB, Alpha=@n37) {constA=1}
  n39 = Multiply(
   A=LinearInterpolate(
    A=VectorParameter() {parameterName=ColorSec, defaultValue=<rgba 0.096 0.1 0.12 1>}.RGB
    B=VectorParameter() {parameterName=ColorMain, defaultValue=<rgba 2.7 0.3249 0.243 0>}.RGB
    Alpha=@n37
   )
   B=@n14
  ) {constA=1}
  MaterialFunctionCall(Emissive=Multiply(
    A=Multiply(A=@n38, B=@n39) {constA=1}
    B=DynamicParameter() {defaultValue=<rgba 0.5 1 1 1>, paramNames [Swipe, EmissiveScale, Param3, Param4]}.EmissiveScale
   )) {materialFunction=&MF_FX_EmissiveAdjustments}
  MaterialFunctionCall(
   Color_Morning=ScalarParameter() {defaultValue=0.5, parameterName=HDRMult_3}
   Color_Day=ScalarParameter() {defaultValue=0.5, parameterName=HDRMult_2}
   Color_Evening=ScalarParameter() {defaultValue=0.5, parameterName=HDRMult_1}
   Color_Night=ScalarParameter() {defaultValue=0.3, parameterName=HDRMult}
  ) {materialFunction=&Color_Time_of_Day}
  MaterialFunctionCall() {materialFunction=&PlatformSwitch_1D}
 ]
}
