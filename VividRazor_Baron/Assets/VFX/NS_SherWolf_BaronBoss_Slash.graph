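# Niagara claw-slash effect for the SherWolf Baron boss, built from short-lived red mesh particles scaled by System.ClawFXSize.
# Wide_Trail bursts four partitioned meshes that spread to per-claw offsets while stretching and fading, with separate color, scale and material parameters for a glow.
# Sharp_Streaks briefly spawns rotating, stretching streaks after a short delay, using two spawn groups copied into visibility tags.
# User.Color feeds only a disabled module.
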NiagaraSystem /VividRazor_Baron/Assets/VFX/NS_SherWolf_BaronBoss_Slash
dependencies {
 usedDataInterfaces [
  &NiagaraDataInterfaceArrayFloat3
  &NiagaraDataInterfaceCurve
  &NiagaraDataInterfaceColorCurve
  &NiagaraDataInterfaceVectorCurve
 ]
 usedDynamicInputs []
 usedModules [
  &self:SystemSpawnScript.SystemScriptSource.SystemScriptGraph.NiagaraNodeAssignment_1.SetVariables_#1
  &SystemState
 ]
}
emitters {
 Sharp_Streaks {
  emitterUpdateScript [
   &EmitterState(Life Cycle Mode=::NewEnumerator1, Loop Behavior=::NewEnumerator1, Loop Delay=0.03, Loop Duration=0.05, UseLoopDelay=true)
   &SpawnRate(SpawnRate=30, Use Spawn Probability=false)
   &SpawnRate(SpawnGroup=1, SpawnRate=40, Use Spawn Probability=false) {module=SpawnRate001}
  ]
  particleSpawnScript [
   /SetVariables(Particles.SpawnGroup=$Emitter.SpawnGroup) {isSetParameters=true}
   /SetVariables(Particles.VisibilityTag=$Particles.SpawnGroup) {isSetParameters=true}
   &InitializeParticle(
    Color=&RandomRangeLinearColor(Minimum=<rgba 1 0.08 0.08 1>, Maximum=<rgba 1 0.12 0.04 1>)
    Lifetime=&UniformRangedFloat(Minimum=0.14, Maximum=0.16)
    Mass Mode=::NewEnumerator1
    Mesh Scale=&VectorFromFloat(Value=&Multiply_Float(A=&RandomRangeFloat(Minimum=8, Maximum=9), B=$System.ClawFXSize))
    Mesh Scale Mode=::NewEnumerator4
    UsePositionOffset=false
    Write Material Random=true
   ) {module=InitializeParticle001}
   &InitialMeshOrientation(
    Mesh Orientation Mode=::NewEnumerator2
    Orientation Vector=<0 0 1>
    Rotation=&MakeVector(X=-0.08, Y=0, Z=0)
    Use Rotation Vector=true
   ) {module=InitialMeshOrientation001}
   /SetVariables(Particles.InitialRotation:Quat4f=&MultiplyQuaternion(
     Quaternion A=$Particles.MeshOrientation
     Quaternion B=&OrientMeshToVector(Look At Direction=<0 0 1>, Reference Axis Vector=<1 0 0>)
    )) {isSetParameters=true}
   &SystemLocation(MaskSpawnGroup=false, Offset=<0 0 -10>)
  ]
  particleUpdateScript [
   &ParticleState()
   &SolveForcesAndVelocity(Clamp Velocity=false, Limit Acceleration=false)
   &ScaleColor(
    Scale Alpha=&FloatFromCurve(
     FloatCurve=DataInterface() {propertyValues {Curve {keys [
         {arriveTangent=6.3626, interpMode=RCIM_Cubic, leaveTangent=6.3626, tangentMode=RCTM_User, time=0, value=0}
         {arriveTangent=-1.8853, interpMode=RCIM_Cubic, leaveTangent=-1.8853, tangentMode=RCTM_User, time=0.12, value=1}
         {interpMode=RCIM_Cubic, leaveTangent=-0.6839, time=1, value=0}
        ]}}}
     Scale Curve=3
    )
    Scale RGB=&VectorFromFloat(Value=&Lerp_Float(
      A=0.1
      B=3
      Alpha=&FloatFromCurve(FloatCurve=DataInterface() {propertyValues {Curve {keys [
           {arriveTangent=-7.4824, interpMode=RCIM_Cubic, leaveTangent=-7.4824, tangentMode=RCTM_User, time=0.1, value=1}
           {arriveTangent=-1.363, interpMode=RCIM_Cubic, leaveTangent=-1.363, tangentMode=RCTM_User, time=0.22, value=0.3}
           {interpMode=RCIM_Cubic, tangentMode=RCTM_User, time=0.7, value=0}
          ]}}})
     ))
    ScaleA=true
    ScaleRGB=true
   )
   &UpdateMeshOrientation(Rotation Rate=&Lerp_Float(
     A=-3
     B=0
     Alpha=&FloatFromCurve(FloatCurve=DataInterface() {propertyValues {Curve {keys [
          {interpMode=RCIM_Cubic, tangentMode=RCTM_User, time=0, value=0}
          {arriveTangent=2.3225, interpMode=RCIM_Cubic, leaveTangent=2.3225, tangentMode=RCTM_User, time=0.8, value=1}
         ]}}})
    )) {module=UpdateMeshOrientation001}
   &ScaleMeshSize(Scale Factor=&Multiply_VectorByFloat(
     Vector=&MakeVector(
      X=&FloatFromCurve(
       FloatCurve=DataInterface() {propertyValues {Curve {keys [
           {arriveTangent=1.3683, interpMode=RCIM_Cubic, leaveTangent=1.3683, tangentMode=RCTM_User, time=0.1, value=0.5}
           {interpMode=RCIM_Cubic, tangentMode=RCTM_User, time=0.7, value=2}
          ]}}}
       Scale Curve=2
      )
      Y=1
      Z=1
     )
     Float=&FloatFromCurve(FloatCurve=DataInterface() {propertyValues {Curve {keys [
          {arriveTangent=1.5926, interpMode=RCIM_Cubic, leaveTangent=1.5926, tangentMode=RCTM_User, time=0, value=0.7}
          {arriveTangent=-0.0424, interpMode=RCIM_Cubic, leaveTangent=-0.0424, tangentMode=RCTM_User, time=0.81, value=1.25}
         ]}}})
    ))
   /SetVariables(Particles.ColorDark:LinearColor=&MultiplyLinearColor(
     Linear Color=$Particles.Color
     Scale Factor=&MakeVector4(
      X=0
      Y=0
      Z=0
      W=&FloatFromCurve(FloatCurve=DataInterface() {propertyValues {Curve {keys [<at 0 1>,<at 0.8 0>]}}})
     )
    )) {isSetParameters=true}
  ]
  rendererClasses [&NiagaraMeshRendererProperties]
  renderers [
   {rendererClass=&NiagaraMeshRendererProperties, rendererIndex=0}
   {rendererClass=&NiagaraMeshRendererProperties, rendererIndex=1}
  ]
 }
 Wide_Trail {
  emitterUpdateScript [
   &EmitterState()
   &SpawnBurst_Instantaneous(Spawn Count=4, Use Spawn Probability=false)
  ]
  particleSpawnScript [
   &PartitionParticles(Number Of Partitions=4)
   &InitializeParticle(
    Color=$User.Color
    Lifetime=0.18
    Mass Mode=::NewEnumerator1
    Mesh Scale=&SelectVectorFromArray(
     Direct Array Index=$Particles.PartitionParticles.Partition
     Vector Selection Array=DataInterface() {propertyValues {FloatData [<20 20 16>, <20 20 20>, <20 20 18>, <20 20 14>]}}
    )
    Mesh Scale Mode=::NewEnumerator4
    Position=&SimulationPosition()
    Position Mode=::NewEnumerator1
    Sprite Size Mode=::NewEnumerator4
    UsePositionOffset=false
    Write Material Random=true
   ) {enabled=false, module=InitializeParticle001}
   &InitializeParticle(
    Lifetime=&RandomRangeFloat(Minimum=0.18, Maximum=0.2)
    Mesh Scale=&Multiply_VectorByFloat(
     Vector=&Multiply_Vector(
      A=&SelectVectorFromArray(
       Direct Array Index=$Particles.PartitionParticles.Partition
       Vector Selection Array=DataInterface() {propertyValues {FloatData [<20 20 16>, <20 20 20>, <20 20 18>, <20 20 14>]}}
      )
      B=&Multiply_Vector(
       A=<0.045 1 1>
       B=&MakeVector(X=&RandomRangeFloat(Minimum=0.6, Maximum=1), Y=1, Z=1)
      )
     )
     Float=$System.ClawFXSize
    )
    Mesh Scale Mode=::NewEnumerator4
    UsePositionOffset=false
    Write Material Random=true
   )
   &InitialMeshOrientation(
    Mesh Axis To Orient=<0 0 1>
    Mesh Orientation Mode=::NewEnumerator2
    Orientation Coordinate Space=::Local
    Rotation=&MakeVector(X=0.2, Y=0, Z=0)
    Use Rotation Vector=true
   )
   &SystemLocation(
    MaskSpawnGroup=false
    Offset=&RotateVectorByQuaternion(
     Vector To Rotate=&Multiply_VectorByFloat(
      Vector=&SelectVectorFromArray(
       Direct Array Index=$Particles.PartitionParticles.Partition
       Vector Selection Array=DataInterface() {propertyValues {FloatData [<80 0 0>, <40 0 0>, <0 0 0>, <-40 0 0>]}}
      )
      Float=$System.ClawFXSize
     )
     Quaternion=$Particles.MeshOrientation
    )
   ) {enabled=false, module=SystemLocation002}
   &CameraOffset(Camera Offset Amount=20)
  ]
  particleUpdateScript [
   &ParticleState()
   /SetVariables(Particles.Position=&Multiply_VectorByFloat(
     Vector=&RotateVectorByQuaternion(
      Vector To Rotate=&Multiply_VectorByFloat(
       Vector=&SelectVectorFromArray(
        Direct Array Index=$Particles.PartitionParticles.Partition
        Vector Selection Array=DataInterface() {propertyValues {FloatData [<80 0 0>, <40 0 0>, <0 0 0>, <-40 0 0>]}}
       )
       Float=$System.ClawFXSize
      )
      Quaternion=$Particles.MeshOrientation
     )
     Float=&FloatFromCurve(FloatCurve=DataInterface() {propertyValues {Curve {keys [
          {arriveTangent=0.4845, interpMode=RCIM_Cubic, leaveTangent=0.4845, tangentMode=RCTM_User, time=0, value=0.1}
          {arriveTangent=3.6537, interpMode=RCIM_Cubic, leaveTangent=3.6537, tangentMode=RCTM_User, time=0.65, value=1}
         ]}}})
    )) {isSetParameters=true}
   &UpdateMeshOrientation(Rotation Rate=&Lerp_Float(
     A=0
     B=1
     Alpha=&FloatFromCurve(FloatCurve=DataInterface() {propertyValues {Curve {keys [{arriveTangent=-1.1894, time=0.2, value=1}, {arriveTangent=-0.7192, leaveTangent=-0.7192, time=1, value=0.1}]}}})
    ))
   &ScaleMeshSize(Scale Factor=&Multiply_VectorByFloat(
     Vector=&MakeVector(
      X=&FloatFromCurve(FloatCurve=DataInterface() {propertyValues {Curve {keys [
           {interpMode=RCIM_Cubic, tangentMode=RCTM_User, time=0, value=1}
           {arriveTangent=5.8758, interpMode=RCIM_Cubic, leaveTangent=5.8758, tangentMode=RCTM_User, time=0.9, value=3}
          ]}}})
      Y=1
      Z=1
     )
     Float=&FloatFromCurve(FloatCurve=DataInterface() {propertyValues {Curve {keys [
          {arriveTangent=0.1307, interpMode=RCIM_Cubic, leaveTangent=0.1307, tangentMode=RCTM_User, time=0, value=0.5}
          {arriveTangent=0.0224, interpMode=RCIM_Cubic, leaveTangent=0.0224, tangentMode=RCTM_User, time=1, value=1.5}
         ]}}})
    ))
   &ScaleColor(
    Scale Alpha=&FloatFromCurve(
     FloatCurve=DataInterface() {propertyValues {Curve {keys [
         {arriveTangent=6.3626, interpMode=RCIM_Cubic, time=0.2, value=1}
         {arriveTangent=-0.1423, interpMode=RCIM_Cubic, leaveTangent=-0.1423, tangentMode=RCTM_User, time=1, value=0}
        ]}}}
     Scale Curve=110
    )
    Scale RGB=&Multiply_VectorByFloat(
     Vector=&MakeVectorFromLinearColorRGB(Color=&ColorFromCurve(ColorCurve=DataInterface() {propertyValues {
         AlphaCurve {keys [<at 0 1 RCIM_Cubic>]}
         BlueCurve {keys [<at 0.0225 0.66 RCIM_Cubic>, <at 0.2022 0.036>, {interpMode=RCIM_Cubic, leaveTangent=-0.0496, time=0.8978, value=0.002}]}
         GreenCurve {keys [<at 0.0225 1.1 RCIM_Cubic>, <at 0.2022 0.0454>, {interpMode=RCIM_Cubic, leaveTangent=-0.0625, time=0.8978, value=0.002}]}
         RedCurve {keys [<at 0.0225 3.3 RCIM_Cubic>, <at 0.2022 0.6>, {interpMode=RCIM_Cubic, leaveTangent=-0.8266, time=0.8978, value=0.1}]}
        }}))
     Float=30
    )
    ScaleA=true
    ScaleRGB=true
   )
   /SetVariables(Particles.Glow Color=&MakeLinearColorFromVectorAndFloat(
     "Vector (RGB)"=&VectorFromCurve(
      VectorCurve=DataInterface() {propertyValues {XCurve {keys [<at 0 1 RCIM_Cubic>,<at 1 0>]}, YCurve {keys [<at 0 1 RCIM_Cubic>,<at 1 0>]}, ZCurve {keys [<at 0 1 RCIM_Cubic>,<at 1 0>]}}}
      Scale Curve=&MakeVectorFromLinearColorRGB(Color=<rgba 2 0.02 0.02 2>)
     )
     "Float (Alpha)"=&FloatFromCurve(
      FloatCurve=DataInterface() {propertyValues {Curve {keys [{interpMode=RCIM_Cubic, tangentMode=RCTM_User, time=0.2, value=1}, {interpMode=RCIM_Cubic, tangentMode=RCTM_User, time=1, value=0}]}}}
      Scale Curve=10
     )
    )) {isSetParameters=true}
   /SetVariables(Particles.ScaleGlow:Vector3f=&Multiply_Vector(A=$Particles.Scale, B=<4.5 1 1>)) {isSetParameters=true}
   &DynamicMaterialParameters(
    Index 0 Param 1=&Lerp_Float(
     A=0.05
     B=0.32
     Alpha=&FloatFromCurve(FloatCurve=DataInterface() {propertyValues {Curve {keys [
          {arriveTangent=0.2245, interpMode=RCIM_Cubic, leaveTangent=0.2245, tangentMode=RCTM_User, time=0, value=0}
          {arriveTangent=0.5171, interpMode=RCIM_Cubic, leaveTangent=0.5171, tangentMode=RCTM_User, time=1, value=1}
         ]}}})
    )
    Index 0 Param 2=&FloatFromCurve(FloatCurve=DataInterface() {propertyValues {Curve {keys [{interpMode=RCIM_Cubic, tangentMode=RCTM_User, time=0.7, value=1}, {interpMode=RCIM_Cubic, tangentMode=RCTM_User, time=1, value=0.15}]}}})
    Module.Index0Float0Write=true
    Module.Index0Float1Write=true
    Module.Index0Float2Write=false
    Module.Index0Float3Write=false
   )
   /SetVariables(Particles.DynamicMaterialParameterGlow:Vector4f=&MakeVector4(
     X=&Lerp_Float(
      A=0.05
      B=0.4
      Alpha=&FloatFromCurve(FloatCurve=DataInterface() {propertyValues {Curve {keys [
           {arriveTangent=0.2881, interpMode=RCIM_Cubic, leaveTangent=0.2881, tangentMode=RCTM_User, time=0.1, value=0}
           {arriveTangent=0.7706, interpMode=RCIM_Cubic, leaveTangent=0.7706, tangentMode=RCTM_User, time=1, value=1}
          ]}}})
     )
     Y=&Lerp_Float(
      A=2
      B=0.3
      Alpha=&FloatFromCurve(FloatCurve=DataInterface() {propertyValues {Curve {keys [
           {arriveTangent=-4.3028, interpMode=RCIM_Cubic, leaveTangent=-4.3028, tangentMode=RCTM_User, time=0, value=1}
           {interpMode=RCIM_Cubic, tangentMode=RCTM_User, time=0.5, value=0}
          ]}}})
     )
     Z=0
     W=0
    )) {isSetParameters=true}
  ]
  rendererClasses [&NiagaraMeshRendererProperties]
  renderers [
   {rendererClass=&NiagaraMeshRendererProperties, rendererIndex=0}
   {rendererClass=&NiagaraMeshRendererProperties, rendererIndex=1}
  ]
 }
}
summary {
 systemName=NS_SherWolf_BaronBoss_Slash
 emitters [
  {emitterName=Wide_Trail, rendererClasses [&NiagaraMeshRendererProperties]}
  {emitterName=Sharp_Streaks, rendererClasses [&NiagaraMeshRendererProperties]}
 ]
 userVariables [
  User.Color {type=LinearColor, description="", defaultValue {struct=LinearColor, value=<rgba 15 0.3 0.3 2>}}
  User.GlowColorScale {type=LinearColor, description="", defaultValue {struct=LinearColor, value=<rgba 2 0.02 0.02 2>}}
 ]
}
